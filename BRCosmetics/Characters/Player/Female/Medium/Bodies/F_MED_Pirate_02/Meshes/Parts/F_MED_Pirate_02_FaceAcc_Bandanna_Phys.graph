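# Physics asset for the bandanna face accessory on the female medium Pirate 02 character, defining simulation shapes for its dangling cloth bones.
# Capsule bodies cover the bandanna base, top and bottom bones, plus the head.
# Constraints join the top and bottom bones to the base with limited swing and locked twist, with the bottom allowed wider swing.
# The head capsule has no constraint to the bandanna chain.
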PhysicsAsset /BRCosmetics/Characters/Player/Female/Medium/Bodies/F_MED_Pirate_02/Meshes/Parts/F_MED_Pirate_02_FaceAcc_Bandanna_Phys
bodies {
 dyn_bandanna_base [{shapeType=Capsule, length=21.8821, radius=3.7676, center=<2.3778 -7.6747 4.7802>, rotation=<rot 39.003 -79.6922 86.3948>}]
 dyn_bandanna_bottom [{shapeType=Capsule, length=2.1837, radius=1, center=<0.1145 -1.0443 -2.0774>, rotation=<rot 0 0 28.6981>}]
 dyn_bandanna_top [{shapeType=Capsule, length=3.7443, radius=2, center=<1.7169 -3.3583 -2.5159>, rotation=<rot 14.4636 23.5853 57.2514>}]
 head [{shapeType=Capsule, length=6.75, radius=4.1076, center=<-4.1422 2.1313 -0.0613>}]
}
constraints [
 {
  bone1Name=dyn_bandanna_bottom
  bone2Name=dyn_bandanna_base
  swing1LimitDegrees=45
  swing2LimitDegrees=30
  twistLimitDegrees=0
  twistMotion=Locked
 }
 {
  bone1Name=dyn_bandanna_top
  bone2Name=dyn_bandanna_base
  swing1LimitDegrees=25
  swing2LimitDegrees=20
  twistLimitDegrees=0.01
  twistMotion=Locked
 }
]
