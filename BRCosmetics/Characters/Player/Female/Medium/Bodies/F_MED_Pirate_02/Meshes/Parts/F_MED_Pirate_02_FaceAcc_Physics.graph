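# Physics asset for the face accessory mesh of the female medium Pirate 02 cosmetic body, covering hair bangs and two feathers.
# Capsule bodies sit on the head, a bangs bone, a feather base and left and right feather bones.
# Constraints form a small chain: both feathers hang from the base, while the base and bangs attach to the head.
# Tight angular limits keep secondary motion subtle, with the base's first swing axis locked.
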PhysicsAsset /BRCosmetics/Characters/Player/Female/Medium/Bodies/F_MED_Pirate_02/Meshes/Parts/F_MED_Pirate_02_FaceAcc_Physics
bodies {
 dyn_bangs_02 [{shapeType=Capsule, length=4.7189, radius=1.7073, center=<2.0043 -0.1309 -3.0387>, rotation=<rot 12.144 -98.9087 -36.6774>}]
 dyn_feather_base [{shapeType=Capsule, length=0.505, radius=0.505, center=<0.6664 -0.2862 -0.5543>, rotation=<rot 0 -47.657 -78.0803>}]
 dyn_feather_left [{shapeType=Capsule, length=15.5802, radius=2.688, center=<1.6275 5.3655 -6.8892>, rotation=<rot -17.924 -21.9187 -34.3284>}]
 dyn_feather_right [{shapeType=Capsule, length=12.7309, radius=2.366, center=<0.0671 7.547 -3.4516>, rotation=<rot -6.2556 -3.3834 126.6567>}]
 head [{shapeType=Capsule, length=6.75, radius=4.1076, center=<-4.1422 2.1313 -0.0613>}]
}
constraints [
 {bone1Name=dyn_feather_left, bone2Name=dyn_feather_base, swing1LimitDegrees=10, swing2LimitDegrees=20, twistLimitDegrees=1}
 {bone1Name=dyn_feather_right, bone2Name=dyn_feather_base, swing1LimitDegrees=10, swing2LimitDegrees=15, twistLimitDegrees=1}
 {bone1Name=dyn_bangs_02, bone2Name=head, swing1LimitDegrees=3, swing2LimitDegrees=3, twistLimitDegrees=2}
 {bone1Name=dyn_feather_base, bone2Name=head, swing1LimitDegrees=2, swing1Motion=Locked, swing2LimitDegrees=2, twistLimitDegrees=0.5}
]
