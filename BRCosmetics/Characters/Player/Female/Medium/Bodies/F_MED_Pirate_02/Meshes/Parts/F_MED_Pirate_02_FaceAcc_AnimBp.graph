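# Post-process style AnimBlueprint for the female Pirate 02 face accessory mesh, adding secondary motion to dangling rings and a bandanna.
# Copies the pose from the attached parent mesh, then chains two Anim Dynamics nodes with linear springs and angular limits on the small and large ring bones.
# Two RigidBody nodes follow with separate physics assets, simulating in spine bone space with overridden gravity; the bandanna pass amplifies component motion.
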AnimBlueprint /BRCosmetics/Characters/Player/Female/Medium/Bodies/F_MED_Pirate_02/Meshes/Parts/F_MED_Pirate_02_FaceAcc_AnimBp
target_skeleton=&/F_MED_Pirate_02_FaceAcc_Skeleton
graphs [
 AnimGraph {nodes [
   n1 = Anim Dynamics - Bone: dyn_ring_small(ComponentPose=Local To Component(LocalPose=Copy Pose From Mesh() {bUseAttachedParent=true}), Alpha=1, ExternalForce="0,0,0") {
    bGravityOverrideInSimSpace=true
    bLinearSpring=true
    bOverrideLinearDamping=true
    linearDampingOverride=0.996
    linearSpringConstant=100
    alphaBoolBlend {blendOption=Linear}
    boundBone {boneName=dyn_ring_small}
    componentAppliedLinearAccClamp=<7500 7500 7500>
    componentLinearAccScale=<1 1 1>
    componentLinearVelScale=<1 1 1>
    physicsBodyDefinitions [{
      sphereCollisionRadius=0
      boundBone {boneName=dyn_ring_small}
      boxExtents=<1 6 6>
      constraintSetup {angularLimitsMax=<0 30 0>, angularLimitsMin=<0 -30 0>, angularTarget=<0 0 0>}
      localJointOffset=<0 0 1.5>
     }]
    relativeSpaceBone {boneName=dyn_ring_small}
   }
   n2 = Anim Dynamics - Bone: dyn_ring_large(ComponentPose=@n1, Alpha=1, ExternalForce="0,0,0") {
    bGravityOverrideInSimSpace=true
    bLinearSpring=true
    bOverrideLinearDamping=true
    linearDampingOverride=0.996
    linearSpringConstant=100
    alphaBoolBlend {blendOption=Linear}
    boundBone {boneName=dyn_ring_large}
    componentAppliedLinearAccClamp=<7500 7500 7500>
    componentLinearAccScale=<1 1 1>
    componentLinearVelScale=<1 1 1>
    physicsBodyDefinitions [{
      sphereCollisionRadius=0
      boundBone {boneName=dyn_ring_large}
      boxExtents=<1 6 8>
      constraintSetup {angularLimitsMax=<0 60 0>, angularLimitsMin=<0 -35 0>, angularTarget=<0 0 0>}
      localJointOffset=<0 0 1>
     }]
    relativeSpaceBone {boneName=dyn_ring_large}
   }
   n3 = RigidBody(ComponentPose=@n2) {
    bClampLinearTranslationLimitToRefPose=true
    bOverrideWorldGravity=true
    overridePhysicsAsset=&/F_MED_Pirate_02_FaceAcc_Physics
    simulationSpace=BaseBoneSpace
    alphaBoolBlend {blendOption=Linear}
    baseBoneRef {boneName=spine_01}
    componentAppliedLinearAccClamp=<500 500 500>
    componentLinearAccScale=<0.5 0.5 0.5>
    componentLinearVelScale=<0.5 0.5 0.5>
    overrideWorldGravity {z=-980}
   }
   n4 = RigidBody(ComponentPose=@n3) {
    bClampLinearTranslationLimitToRefPose=true
    bOverrideWorldGravity=true
    overridePhysicsAsset=&/F_MED_Pirate_02_FaceAcc_Bandanna_Phys
    simulationSpace=BaseBoneSpace
    alphaBoolBlend {blendOption=Linear}
    baseBoneRef {boneName=spine_02}
    componentAppliedLinearAccClamp=<1000 1000 1000>
    componentLinearAccScale=<5 5 5>
    componentLinearVelScale=<3 3 3>
    overrideWorldGravity {z=-980}
   }
   Output Pose(Result=Component To Local(ComponentPose=@n4)) {layerGroup=None}
  ]}
 EventGraph {nodes [
   TryGetPawnOwner()
   Event BlueprintUpdateAnimation()
  ]}
]
variables []
